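# Translucent lens flare material, named for the Sea Mist lighthouse, that fades in and out as if from a rotating beam.
# Opacity combines a rotated flare texture, camera and scene depth fades, a vertical view-angle mask, and a horizontal mask comparing camera direction to a time-spun local axis.
# World position offset aligns the mesh to camera and pulls it toward the viewer, collapsing it to its pivot when the spin mask nears zero.
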Material /Hera/Sets/SeaMist/FX/M_LenseFlare_LightHouse
graph {
 properties {
  bUsedWithNiagaraSprites=true
  bUsedWithParticleSprites=true
  blendMode=BLEND_Translucent
  pixelDepthOffsetMode=PDOM_Legacy
  usedShadingModels=Default Lit
 }
 nodes [
  n1 = VectorParameter() {parameterName=Color Alpha, defaultValue=<rgba 1 1 1 1>}
  n2 = TextureSampleParameter2D(UVs=MaterialFunctionCall(UVs=TextureCoordinate(), "Rotation Angle (0-1)"=ScalarParameter() {defaultValue=0.25, parameterName=Texture Rotation}) {materialFunction=&CustomRotator}.Rotated Values) {parameterName=FlareTexture, samplerType=SAMPLERTYPE_Grayscale, texture=/Game/Effects/Fort_Effects/Textures/T_EFX_CameraFlare_01}
  n3 = ScalarParameter() {parameterName="Fade Length (S)"}
  n4 = ScalarParameter() {parameterName="Fade Offset (S)"}
  n5 = Clamp(None=Multiply(
    A=Multiply(A=@n1.A, B=@n2.G)
    B=MaterialFunctionCall(Fade Length=@n3, Fade Offset=@n4) {materialFunction=&CameraDepthFade}.Result
   ))
  n6 = Subtract(
   A=CameraPositionWS()
   B=MaterialFunctionCall() {materialFunction=&ComponentPivotLocation}.Result
  )
  n7 = Abs(None=DotProduct(A=Normalize(VectorInput=@n6), B=Constant3Vector() {constant=<rgba 0 0 1 0>}))
  n8 = VectorParameter() {parameterName=Flare Dot verticla Options, channelNames {a=contrast, b=add, g=mul, r=pow}, defaultValue=<rgba 1 1 0 0>}
  n9 = Multiply(A=Power(Base=OneMinus(None=@n7), Exp=@n8.pow), B=@n8.mul)
  n10 = Saturate(None=MaterialFunctionCall(In=Add(A=@n9, B=@n8.add), Contrast=@n8.contrast) {materialFunction=&CheapContrast}.Result)
  n11 = Constant3Vector() {constant=<rgba 1 1 0 0>}
  n12 = Multiply(A=@n6, B=@n11)
  n13 = Normalize(VectorInput=Transform(None=Constant3Vector() {constant=<rgba 0 0 1 0>}) {transformSourceType=TRANSFORMSOURCE_Local})
  n14 = Time()
  n15 = ScalarParameter() {group=WPO, parameterName=WPO Time Spin Offset 1}
  n16 = Add(
   A=Multiply(A=Add(A=@n14, B=@n15), B=ScalarParameter() {defaultValue=0.01, group=WPO, parameterName=WPO Spin Speed})
   B=ScalarParameter() {group=WPO, parameterName=WPO Time Spin Offset 2}
  )
  n17 = Constant3Vector() {constant=<rgba 1 0 0 0>}
  n18 = Multiply(
   A=RotateAboutAxis(
    NormalizedRotationAxis=NamedRerouteDeclaration(None=@n13) {name=Rot Axis, nodeColor=<rgba 1 0.7031 0 1>}
    RotationAngle=NamedRerouteDeclaration(None=@n16) {name=Rot Angle, nodeColor=<rgba 0.3828 0 1 1>}
    PivotPoint=MaterialFunctionCall() {materialFunction=&ComponentPivotLocation}.Result
    Position=Transform(None=@n17) {transformSourceType=TRANSFORMSOURCE_Local}
   )
   B=@n11
  )
  n19 = Abs(None=DotProduct(A=Normalize(VectorInput=@n12), B=Normalize(VectorInput=@n18)))
  n20 = VectorParameter() {parameterName=Flare Dor Options, channelNames {a=contrast, b=add, g=mul, r=pow}, defaultValue=<rgba 1 1 0 0>}
  n21 = Multiply(A=Power(Base=OneMinus(None=@n19), Exp=@n20.pow), B=@n20.mul)
  n22 = Saturate(None=MaterialFunctionCall(In=Add(A=@n21, B=@n20.add), Contrast=@n20.contrast) {materialFunction=&CheapContrast}.Result)
  DepthFade(Opacity=Multiply(A=Multiply(A=@n5, B=@n10), B=@n22), FadeDistance=ScalarParameter() {parameterName=Depth fade}) {fadeDistanceDefault=50}
  n23 = ScalarParameter() {defaultValue=1000, parameterName=Cam Offset}
  n24 = MaterialFunctionCall() {materialFunction=&ComponentPivotLocation}
  n25 = WorldPosition() {worldPositionShaderOffset=WPT_ExcludeAllShaderOffsets}
  n26 = Saturate(None=Subtract(A=@n22) {constB=0.001})
  LinearInterpolate(
   A=Subtract(
    A=MaterialFunctionCall() {materialFunction=&ComponentPivotLocation}.Result
    B=WorldPosition().XYZ
   )
   B=Add(
    A=MaterialFunctionCall(Offset Amount=@n23, World Position=@n24.Result) {materialFunction=/Engine/Functions/Engine_MaterialFunctions02/WorldPositionOffset/CameraOffset}.Clamped Camera Offset
    B=MaterialFunctionCall!(WorldPosition=@n25.XYZ, PivotLocation, WS Normals, Custom Object Basis 1, Custom Object Basis 2, Custom Object Basis 3) {materialFunction=/Engine/Functions/Engine_MaterialFunctions02/WorldPositionOffset/AlignMeshToTheCamera}.World Position Offset
   )
   Alpha=Saturate(None=Ceil(None=@n26))
  )
  MaterialFunctionCall(Emissive=@n1.RGB) {materialFunction=&MF_FX_EmissiveAdjustments}
  ObjectPositionWS()
 ]
}
