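# Unlit, two-sided alpha-composite FX material for the Squidwharf lighthouse light, combining camera-facing glow with a spinning beam.
# Red vertex color splits camera-aligned corona and airglow bands from the pivot-rotated beam, shaped by fresnel, gradient and panning noise.
# Camera angle against a time-rotated axis drives a brightness boost, alongside depth fades and time-of-day scaling, whose inputs all read Morning.
# Quality and platform switches simplify glow, offsets and fading.
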Material /Hera/Sets/SeaMist/FX/M_Squidwharf_LightHouse
graph {
 properties {
  bEnableResponsiveAA=true
  bUseMaterialAttributes=true
  bUsedWithMeshParticles=true
  bUsedWithNanite=true
  bUsedWithNiagaraMeshParticles=true
  bUsedWithNiagaraSprites=true
  blendMode=BLEND_AlphaComposite
  pixelDepthOffsetMode=PDOM_Legacy
  shadingModel=MSM_Unlit
  twoSided=true
  usedShadingModels=Unlit
 }
 nodes [
  MaterialFunctionCall() {materialFunction=&PlatformSwitch_1D}
  n1 = ScalarParameter() {defaultValue=4.5, group=Fresnel Rotation, parameterName=FresnelExponentHigh}
  n2 = ScalarParameter() {defaultValue=1, group=Fresnel Rotation, parameterName=FresnelExponentLow}
  n3 = ScalarParameter() {defaultValue=0.9, group=Gradient rotation, parameterName=FalloffRadius}
  n4 = Divide(
   A=ScalarParameter() {defaultValue=300, group=Gradient rotation, parameterName=Length}
   B=ScalarParameter() {defaultValue=400, group=Gradient rotation, parameterName=FalloffBaseLength}
  )
  n5 = SphereMask(
   A=ComponentMask(None=TextureCoordinate()) {g=true}
   B=Constant() {r=1}
   Radius=Saturate(None=Divide(A=@n3, B=@n4))
   Hardness=ScalarParameter() {defaultValue=0.025, group=Gradient rotation, parameterName=FalloffHardness}
  )
  n6 = Power(Base=@n5, Exp=ScalarParameter() {defaultValue=3.5, group=Gradient rotation, parameterName=MaskExponent})
  n7 = NamedRerouteUsage()
  n8 = NamedRerouteUsage()
  n9 = VertexNormalWS()
  n10 = LinearInterpolate(
   A=ScalarParameter() {group=Fresnel Rotation, parameterName=fresnel Low}
   B=ScalarParameter() {defaultValue=1, group=Fresnel Rotation, parameterName=fresnel High}
   Alpha=MaterialFunctionCall(
    Brightness=ScalarParameter() {defaultValue=1, group=Fresnel Rotation, parameterName=FresnelBrightness}
    Exponent=LinearInterpolate(A=@n1, B=@n2, Alpha=@n6)
    Vertex / Pixel Normal=MaterialFunctionCall(Rotation Axis=@n7, Rotation Angle=@n8, World Space Vertex Normals=@n9) {materialFunction=&FixRotateAboutAxisNormals}.New World Space Vertex Normal
   ) {materialFunction=&MF_Fresnel_FX}.Fresnel Double Sided
  )
  n11 = LinearInterpolate(B=ScalarParameter() {defaultValue=0.6, group=Gradient rotation, parameterName=OpacityHigh}, Alpha=@n6)
  n12 = ScalarParameter() {group=Texture, parameterName=Texture Low}
  n13 = ScalarParameter() {defaultValue=1, group=Texture, parameterName=Texture High}
  n14 = TextureCoordinate()
  n15 = VectorParameter() {group=Texture, parameterName=NoiseUVTileSpeedA, defaultValue=<rgba 1 0 0.018 0>}
  n16 = ComponentMask(None=@n15.RGB) {g=true, r=true}
  n17 = TextureObjectParameter() {
   group=Texture
   parameterName=Noise Tex RG
   samplerType=SAMPLERTYPE_LinearColor
   texture=&T_NoisePack_LinearColor
  }
  n18 = ScalarParameter() {parameterName=Noisue Tex Mip Bias}
  n19 = TextureSample!(
   UVs=Panner(Coordinate=Multiply(A=@n14, B=@n16), Speed=ComponentMask(None=@n15.RGBA) {a=true, b=true}) {bFractionalPart=true}
   Tex=@n17
   Bias=@n18
   Apply View MipBias
  ) {mipValueMode=TMVM_MipBias, samplerType=SAMPLERTYPE_LinearColor, texture=&T_NoisePack_LinearColor}
  n20 = TextureCoordinate()
  n21 = VectorParameter() {group=Texture, parameterName=NoiseUVTileSpeedB, defaultValue=<rgba 3 0.11 0.01 0>}
  n22 = ComponentMask(None=@n21.RGB) {g=true, r=true}
  n23 = TextureSample(
   UVs=Panner(Coordinate=Multiply(A=@n20, B=@n22), Speed=ComponentMask(None=@n21.RGBA) {a=true, b=true}) {bFractionalPart=true, speedX=-0.026}
   Tex=@n17
   Bias=@n18
  ) {mipValueMode=TMVM_MipBias, samplerType=SAMPLERTYPE_LinearColor, texture=&T_NoisePack_LinearColor}
  n24 = Add(A=Subtract(A=@n19.R) {constB=0.5}, B=Subtract(A=@n23.G) {constB=0.5})
  n25 = Multiply(
   A=Multiply(A=OneMinus(None=@n6), B=LinearInterpolate(A=@n12, B=@n13, Alpha=@n24))
   B=ScalarParameter() {defaultValue=1.5, group=Texture, parameterName=NoiseScaleMul}
  )
  n26 = Multiply(A=Add(A=Multiply(A=@n10, B=@n11), B=Multiply(A=@n10, B=@n25)), B=Power(Base=@n5, Exp=ScalarParameter() {parameterName=Mask Exponent 2}))
  NamedRerouteDeclaration(None=@n26) {name=Roation Opacity, nodeColor=<rgba 0 0.1562 1 1>}
  NamedRerouteDeclaration(None=Saturate(None=@n26)) {name=Color Lerp alpha, nodeColor=<rgba 0.2422 0 1 1>}
  n27 = ScalarParameter() {defaultValue=0.35, group=Face camer Brightness, parameterName=Emissive low Hor}
  n28 = ScalarParameter() {defaultValue=2, group=Face camer Brightness, parameterName=Emissive High Hor}
  n29 = Subtract(
   A=CameraPositionWS()
   B=MaterialFunctionCall() {materialFunction=&ComponentPivotLocation}.Result
  )
  n30 = Constant3Vector() {constant=<rgba 1 1 0 0>}
  n31 = Multiply(A=@n29, B=@n30)
  n32 = Transform(None=Constant3Vector() {constant=<rgba 0 0 1 0>}) {transformSourceType=TRANSFORMSOURCE_Local}
  n33 = Multiply(
   A=Add(A=Time(), B=ScalarParameter() {group=Face camer Brightness, parameterName=WPO Time Spin Offset 1 2})
   B=ScalarParameter() {defaultValue=0.06, group=Face camer Brightness, parameterName=WPO Spin Speed 2}
  )
  n34 = ScalarParameter() {group=Face camer Brightness, parameterName=WPO Time Spin Offset 2 2}
  n35 = Constant3Vector() {constant=<rgba 1 0 0 0>}
  n36 = Multiply(
   A=RotateAboutAxis(
    NormalizedRotationAxis=Normalize(VectorInput=@n32)
    RotationAngle=Add(A=@n33, B=@n34)
    PivotPoint=MaterialFunctionCall() {materialFunction=&ComponentPivotLocation}.Result
    Position=Transform(None=@n35) {transformSourceType=TRANSFORMSOURCE_Local}
   )
   B=@n30
  )
  n37 = Abs(None=DotProduct(A=Normalize(VectorInput=@n31), B=Normalize(VectorInput=@n36)))
  n38 = VectorParameter() {
   group=Face camer Brightness
   parameterName=Emissive mul options hor
   channelNames {a=contrast, b=add, g=mul, r=pow}
   defaultValue=<rgba 1 1 -0.1 2>
  }
  n39 = Multiply(A=Power(Base=OneMinus(None=@n37), Exp=@n38.pow), B=@n38.mul)
  n40 = Saturate(None=MaterialFunctionCall(In=Add(A=@n39, B=@n38.add), Contrast=@n38.contrast) {materialFunction=&CheapContrast}.Result)
  n41 = Abs(None=DotProduct(A=Normalize(VectorInput=@n29), B=Constant3Vector() {constant=<rgba 0 0 1 0>}))
  n42 = VectorParameter() {
   group=Face camer Brightness
   parameterName=Emissive mul options 2 ver
   channelNames {a=contrast, b=add, g=mul, r=pow}
   defaultValue=<rgba 1 1 -0.25 2>
  }
  n43 = Power(Base=LinearInterpolate(Alpha=OneMinus(None=@n41)), Exp=@n42.pow)
  n44 = MaterialFunctionCall(In=Add(A=Multiply(A=@n43, B=@n42.mul), B=@n42.add), Contrast=@n42.contrast) {materialFunction=&CheapContrast}
  NamedRerouteDeclaration(None=LinearInterpolate(
    A=ScalarParameter() {defaultValue=0.35, group=Face camer Brightness, parameterName=Emissive low Hor 2}
    B=LinearInterpolate(A=@n27, B=@n28, Alpha=@n40)
    Alpha=Saturate(None=@n44.Result)
   )) {name=Light house face camera boost, nodeColor=<rgba 0.8047 1 0 1>}
  n45 = ScalarParameter() {defaultValue=0.1, group=Corona, parameterName=Corona Width}
  n46 = ScalarParameter() {group=Corona, parameterName=Corona shaprness}
  n47 = ScalarParameter() {group=Corona, parameterName=Colorna Offset}
  n48 = Multiply(
   A=Power(
    Base=MaterialFunctionCall!(Width=@n45, Sharpness=@n46, Offset=@n47, Direction Switch, Compare, Input Coordinates) {materialFunction=/Engine/Functions/Engine_MaterialFunctions03/Procedurals/GeneratedBand}.Result
    Exp=ScalarParameter() {defaultValue=1, group=Corona, parameterName=Corona Pow}
   )
   B=ScalarParameter() {defaultValue=1, group=Corona, parameterName=Corona Opacity Scale}
  )
  n49 = Constant()
  n50 = ComponentMask(None=Add(A=TextureCoordinate(), B=ScalarParameter() {group=Airglow, parameterName=Airglow Offset})) {g=true}
  n51 = Constant() {r=0.5}
  n52 = ScalarParameter() {defaultValue=0.1, group=Airglow, parameterName=Airglow Width}
  n53 = ScalarParameter() {group=Airglow, parameterName=Airglow shaprness}
  n54 = Multiply(
   A=Power(
    Base=SphereMask(A=@n50, B=@n51, Radius=@n52, Hardness=@n53)
    Exp=ScalarParameter() {defaultValue=1, group=Airglow, parameterName=Airglow Pow}
   )
   B=ScalarParameter() {defaultValue=1, group=Airglow, parameterName=Airglow Opacity Scale}
  )
  n55 = MaterialFunctionCall("Xb1,PS4"=@n49, "PC-High&&Epic,XSX,PS5"=@n54, "PC-Perf,Android"=@n49, Switch=@n49, "PC-Low(NotPerf)&&Medium"=@n54) {materialFunction=&PlatformSwitch_1D}
  n56 = Add(A=TextureCoordinate(), B=ScalarParameter() {group=Airglow 2, parameterName=Airglow Offset 2})
  n57 = Power(
   Base=SphereMask(
    A=ComponentMask(None=@n56) {g=true}
    B=Constant() {r=0.5}
    Radius=ScalarParameter() {defaultValue=0.1, group=Airglow 2, parameterName=Airglow Width 2}
    Hardness=ScalarParameter() {group=Airglow 2, parameterName=Airglow shaprness 2}
   )
   Exp=ScalarParameter() {defaultValue=1, group=Airglow 2, parameterName=Airglow Pow 2}
  )
  n58 = ScalarParameter() {defaultValue=1, group=Airglow 2, parameterName=Airglow Opacity Scale 2}
  n59 = Add(A=Add(A=Add(A=@n48, B=@n55.Result), B=Multiply(A=@n57, B=@n58)) {constB=0}) {constB=0}
  n60 = Constant()
  NamedRerouteDeclaration(None=QualitySwitch(Default=@n59, Low=@n60, High=@n59, Medium=@n60, Epic=@n59)) {name=Camera Faceing Opacity, nodeColor=<rgba 0.7344 0 1 1>}
  n61 = Normalize(VectorInput=Transform(None=VectorParameter() {group=WPO rotation, parameterName=WPO spin Axis, defaultValue=<rgba 0 0 1 1>}.RGB) {transformSourceType=TRANSFORMSOURCE_Local})
  n62 = Add(
   A=Multiply(A=Time(), B=ScalarParameter() {group=WPO rotation, parameterName=WPO Spin Speed})
   B=ScalarParameter() {group=WPO rotation, parameterName=WPO Time Spin Offset}
  )
  NamedRerouteDeclaration(None=RotateAboutAxis(
    NormalizedRotationAxis=NamedRerouteDeclaration(None=@n61) {name=Rot Axis, nodeColor=<rgba 1 0.7031 0 1>}
    RotationAngle=NamedRerouteDeclaration(None=@n62) {name=Rot Angle, nodeColor=<rgba 0.3828 0 1 1>}
    PivotPoint=MaterialFunctionCall() {materialFunction=&ComponentPivotLocation}.Result
    Position=WorldPosition().XYZ
   )) {name=WPO Spinning Light, nodeColor=<rgba 0 1 0.4141 1>}
  QualitySwitch()
  n63 = TextureCoordinate()
  n64 = ScalarParameter() {group=Core, parameterName=Core Offset}
  n65 = SphereMask(
   A=ComponentMask(None=Add(A=@n63, B=@n64)) {g=true}
   B=Constant() {r=0.5}
   Radius=ScalarParameter() {defaultValue=0.1, group=Core, parameterName=Core Width}
   Hardness=ScalarParameter() {group=Core, parameterName=Core shaprness}
  )
  n66 = ScalarParameter() {defaultValue=1, group=Core, parameterName=Core Pow}
  Reroute(None=Multiply(A=Power(Base=@n65, Exp=@n66), B=ScalarParameter() {defaultValue=1, group=Core, parameterName=Core Opacity Scale})) {desc="not needed in new version\r\nit seems, has weird visual  issues\r\ntoo for some reason"}
  n67 = VectorParameter() {group=colors, parameterName=Color A, defaultValue=<rgba 1 1 0 0>}
  n68 = LinearInterpolate(A=VectorParameter() {group=colors, parameterName=Color b, defaultValue=<rgba 1 0.5 0 0>}.RGB, B=@n67.RGB, Alpha=NamedRerouteUsage())
  n69 = VertexColor()
  n70 = VectorParameter() {group=Opacity, parameterName=TOD Scales, channelNames {a=Night, b=Evening, g=Day, r=Morning}, defaultValue=<rgba 1 1 1 1>}
  n71 = Multiply(
   A=LinearInterpolate(A=NamedRerouteUsage(), B=NamedRerouteUsage(), Alpha=VertexColor().R) {constB=0.2, desc="A - cam face\r\nb - Rotate"}
   B=MaterialFunctionCall(
    Fade Length=ScalarParameter() {group=Camera Depth Fade, parameterName="Cam Fade Length (S)"}
    Fade Offset=ScalarParameter() {group=Camera Depth Fade, parameterName="Cam Fade Offset (S)"}
   ) {materialFunction=&CameraDepthFade}.Result
  )
  n72 = DepthFade(Opacity=@n71, FadeDistance=ScalarParameter() {defaultValue=100, group=Opacity, parameterName=Fade Distance})
  n73 = Multiply(
   A=Multiply(
    A=MaterialFunctionCall(Value_Morning=@n70.Morning, Value_Day=@n70.Morning, Value_Evening=@n70.Morning, Value_Night=@n70.Morning) {materialFunction=&Scalar_Time_of_Day}.Result
    B=MaterialFunctionCall("Xb1,PS4"=@n72, "PC-High&&Epic,XSX,PS5"=@n72, "PC-Perf,Android"=@n71, Switch=@n71, "PC-Low(NotPerf)&&Medium"=@n72) {materialFunction=&PlatformSwitch_1D}.Result
   )
   B=LinearInterpolate(B=NamedRerouteUsage(), Alpha=VertexColor().R) {constA=1}
  )
  n74 = ScalarParameter() {defaultValue=1, parameterName=Fade color}
  n75 = WorldPosition()
  n76 = MaterialFunctionCall!(
   WorldPosition=@n75.XYZ
   PivotLocation=MaterialFunctionCall() {materialFunction=&ComponentPivotLocation}.Result
   WS Normals
   Custom Object Basis 1
   Custom Object Basis 2
   Custom Object Basis 3
  ) {materialFunction=/Engine/Functions/Engine_MaterialFunctions02/WorldPositionOffset/AlignMeshToTheCamera}
  n77 = ScalarParameter() {group=Cam offset, parameterName=Camera Offset}
  n78 = ScalarParameter() {group=Cam offset, parameterName=Camera Offset clamp}
  n79 = Add(A=@n76.World Position Offset, B=@n75.XYZ)
  n80 = VertexColor()
  n81 = LinearInterpolate(
   A=Add(
    A=@n76.World Position Offset
    B=MaterialFunctionCall(Offset Amount=@n77, Clamp Padding=@n78, World Position=@n79) {materialFunction=/Engine/Functions/Engine_MaterialFunctions02/WorldPositionOffset/CameraOffset}.Clamped Camera Offset
   )
   B=NamedRerouteUsage()
   Alpha=@n80.R
  ) {constB=0, desc="A - cam face\r\nb - Rotate"}
  n82 = Constant() {r=-1}
  n83 = Subtract(A=WorldPosition().XYZ, B=ObjectPositionWS())
  n84 = Add(A=NamedRerouteUsage(), B=LinearInterpolate(A=Multiply(A=@n82, B=@n83), Alpha=@n80.R) {constB=0})
  SetMaterialAttributes!(
   MaterialAttributes
   Emissive Color=Multiply(A=LinearInterpolate(A=@n67.RGB, B=@n68, Alpha=@n69.R), B=Multiply(A=@n73, B=@n74))
   Opacity=Multiply(A=@n73, B=ScalarParameter() {defaultValue=1, parameterName=Fade Opac})
   World Position Offset=QualitySwitch(Default=@n81, Low=@n84, High=@n81, Medium=@n84, Epic=@n81)
  ) {
   desc=Old
   attributeSetTypes [#1, #2, #3]
  }
  ShadingPathSwitch()
 ]
}
