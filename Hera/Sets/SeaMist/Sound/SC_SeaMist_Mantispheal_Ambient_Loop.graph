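# Looping ambient sound cue for the SeaMist Mantispheal, playing intermittent moans under the loot ambient loop sound class.
# Each loop waits a random 1–4 seconds, then picks one of four equally weighted groups and a wave within it.
# Groups differ in size, so moans in smaller groups play more often.
# It references shared item-ambient attenuation and loot-ambient concurrency assets, with bus and submix sends disabled.
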SoundCue /Hera/Sets/SeaMist/Sound/SC_SeaMist_Mantispheal_Ambient_Loop
root=Looping_3
nodes [
 {id=Delay_6, children [Random_6], properties {delayMax=4, delayMin=1}}
 {id=Looping_3, children [Delay_6]}
 {id=Random_2, children [WavePlayer_27, WavePlayer_28, WavePlayer_29, WavePlayer_30, WavePlayer_31], properties {weights [1,1,1,1,1]}}
 {id=Random_3, children [WavePlayer_37, WavePlayer_33, WavePlayer_34], properties {weights [1,1,1]}}
 {id=Random_4, children [WavePlayer_39, WavePlayer_40], properties {weights [1,1]}}
 {id=Random_5, children [WavePlayer_41, WavePlayer_43, WavePlayer_42], properties {weights [1,1,1]}}
 {id=Random_6, children [Random_2,Random_3,Random_4,Random_5], properties {weights [1,1,1,1]}}
 {id=WavePlayer_27, properties {soundWaveAssetPtr=/Hera/Sets/SeaMist/Sound/Wav/SC_SeaMist_Mantispheal_Ambient_Moan_01}}
 {id=WavePlayer_28, properties {soundWaveAssetPtr=/Hera/Sets/SeaMist/Sound/Wav/SC_SeaMist_Mantispheal_Ambient_Moan_02}}
 {id=WavePlayer_29, properties {soundWaveAssetPtr=/Hera/Sets/SeaMist/Sound/Wav/SC_SeaMist_Mantispheal_Ambient_Moan_03}}
 {id=WavePlayer_30, properties {soundWaveAssetPtr=/Hera/Sets/SeaMist/Sound/Wav/SC_SeaMist_Mantispheal_Ambient_Moan_04}}
 {id=WavePlayer_31, properties {soundWaveAssetPtr=/Hera/Sets/SeaMist/Sound/Wav/SC_SeaMist_Mantispheal_Ambient_Moan_05}}
 {id=WavePlayer_33, properties {soundWaveAssetPtr=/Hera/Sets/SeaMist/Sound/Wav/SC_SeaMist_Mantispheal_Ambient_Moan_07}}
 {id=WavePlayer_34, properties {soundWaveAssetPtr=/Hera/Sets/SeaMist/Sound/Wav/SC_SeaMist_Mantispheal_Ambient_Moan_08}}
 {id=WavePlayer_37, properties {soundWaveAssetPtr=/Hera/Sets/SeaMist/Sound/Wav/SC_SeaMist_Mantispheal_Ambient_Moan_11}}
 {id=WavePlayer_39, properties {soundWaveAssetPtr=/Hera/Sets/SeaMist/Sound/Wav/SC_SeaMist_Mantispheal_Ambient_Moan_13}}
 {id=WavePlayer_40, properties {soundWaveAssetPtr=/Hera/Sets/SeaMist/Sound/Wav/SC_SeaMist_Mantispheal_Ambient_Moan_14}}
 {id=WavePlayer_41, properties {soundWaveAssetPtr=/Hera/Sets/SeaMist/Sound/Wav/SC_SeaMist_Mantispheal_Ambient_Moan_15}}
 {id=WavePlayer_42, properties {soundWaveAssetPtr=/Hera/Sets/SeaMist/Sound/Wav/SC_SeaMist_Mantispheal_Ambient_Moan_16}}
 {id=WavePlayer_43, properties {soundWaveAssetPtr=/Hera/Sets/SeaMist/Sound/Wav/SC_SeaMist_Mantispheal_Ambient_Moan_17}}
]
properties {
 attenuationSettings=/Game/Sounds/Attenuation/Fort_Item_Ambient_Loop_Mono_Default_Att
 bEnableBusSends=false
 bEnableSubmixSends=false
 maxDistance=1250
 priority=100
 soundClassObject=/Game/Sounds/FortSoundClassesAndModes/Fort_Loot_Ambient_Loop
 volumeMultiplier=2
 attenuationOverrides {
  bEnableFocusInterpolation=true
  bEnableListenerFocus=true
  distanceAlgorithm=Logarithmic
  falloffDistance=1100
  focusAttackInterpSpeed=20
  focusAzimuth=20
  focusReleaseInterpSpeed=10
  focusVolumeAttenuation=3.5
  lPFFrequencyAtMax=10000
  lPFRadiusMax=1000
  lPFRadiusMin=500
  nonFocusAzimuth=40
  nonFocusVolumeAttenuation=0.8
  nonSpatializedRadiusStart=200
  occlusionInterpolationTime=0.3
  occlusionLowPassFilterFrequency=3000
  occlusionVolumeAttenuation=0.75
  spatializationAlgorithm=SPATIALIZATION_HRTF
  attenuationShapeExtents {x=0}
  customAttenuationCurve {editorCurveData {keys [<at 0 1 RCIM_Cubic>, {arriveTangent=-1, interpMode=RCIM_Cubic, leaveTangent=-1, time=0.5, value=0.1}, <at 1 0 RCIM_Cubic>]}}
 }
 concurrencyOverrides {resolutionRule=StopQuietest}
 concurrencySet [/Game/Sounds/Concurrency_Rules/Loot_Ambient_Conc]
}
