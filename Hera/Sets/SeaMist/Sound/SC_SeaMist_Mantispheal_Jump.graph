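# Sound cue for the jump of the SeaMist Mantispheal, classed as an ambient prop sound.
# A Mixer layers one of three random boings with one of eight random moans drawn from jump and ambient recordings, with the moan slightly louder.
# A root Modulator holds volume fixed, leaving pitch at its default range.
# A dedicated attenuation asset handles falloff, and a concurrency override blocks new plays at the limit.
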SoundCue /Hera/Sets/SeaMist/Sound/SC_SeaMist_Mantispheal_Jump
root=Modulator_1
nodes [
 {id=Mixer_0, children [Random_0,Random_9], properties {inputVolume [1,1.25]}}
 {id=Modulator_1, children [Mixer_0], properties {volumeMax=1, volumeMin=1}}
 {id=Random_0, children [WavePlayer_2,WavePlayer_4,WavePlayer_5], properties {weights [1,1,1]}}
 {
  id=Random_9
  children [WavePlayer_20, WavePlayer_21, WavePlayer_22, WavePlayer_23, WavePlayer_35, WavePlayer_36, WavePlayer_38, WavePlayer_32]
  properties {weights [1,1,1,1,1,1,1,1]}
 }
 {id=WavePlayer_2, properties {soundWaveAssetPtr=/Hera/Sets/SeaMist/Sound/Wav/SC_SeaMist_Mantispheal_Jump_Boing_01}}
 {id=WavePlayer_20, properties {soundWaveAssetPtr=/Hera/Sets/SeaMist/Sound/Wav/SC_SeaMist_Mantispheal_Jump_Moan_01}}
 {id=WavePlayer_21, properties {soundWaveAssetPtr=/Hera/Sets/SeaMist/Sound/Wav/SC_SeaMist_Mantispheal_Jump_Moan_02}}
 {id=WavePlayer_22, properties {soundWaveAssetPtr=/Hera/Sets/SeaMist/Sound/Wav/SC_SeaMist_Mantispheal_Jump_Moan_03}}
 {id=WavePlayer_23, properties {soundWaveAssetPtr=/Hera/Sets/SeaMist/Sound/Wav/SC_SeaMist_Mantispheal_Jump_Moan_04}}
 {id=WavePlayer_32, properties {soundWaveAssetPtr=/Hera/Sets/SeaMist/Sound/Wav/SC_SeaMist_Mantispheal_Ambient_Moan_06}}
 {id=WavePlayer_35, properties {soundWaveAssetPtr=/Hera/Sets/SeaMist/Sound/Wav/SC_SeaMist_Mantispheal_Ambient_Moan_09}}
 {id=WavePlayer_36, properties {soundWaveAssetPtr=/Hera/Sets/SeaMist/Sound/Wav/SC_SeaMist_Mantispheal_Ambient_Moan_10}}
 {id=WavePlayer_38, properties {soundWaveAssetPtr=/Hera/Sets/SeaMist/Sound/Wav/SC_SeaMist_Mantispheal_Ambient_Moan_12}}
 {id=WavePlayer_4, properties {soundWaveAssetPtr=/Hera/Sets/SeaMist/Sound/Wav/SC_SeaMist_Mantispheal_Jump_Boing_02}}
 {id=WavePlayer_5, properties {soundWaveAssetPtr=/Hera/Sets/SeaMist/Sound/Wav/SC_SeaMist_Mantispheal_Jump_Boing_03}}
]
properties {
 attenuationSettings=&/ATT_SeaMist_Mantispheal_Jump
 bOverrideConcurrency=true
 maxDistance=6250
 soundClassObject=&Fort_Ambient_Props_Athena
 volumeMultiplier=1.35
 attenuationOverrides {falloffDistance=1000, stereoSpread=500, attenuationShapeExtents {x=500}}
 concurrencyOverrides {resolutionRule=PreventNew}
}
